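; Conversation component for hireable NPCs that picks the icon shown on the NPC's dialog marker.
; On BeginPlay, if brush adjustment is enabled, it checks whether the owning actor has the NPC.Status.NoHire gameplay tag.
; NPCs with that tag get the speech indicator icon, while all others get the hire indicator icon.
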(event EventBeginPlay
  (if (Variables|Default|GetAdjustBrushForNoHireTag)
    (bind _returnvalue (GameplayTags|HasMatchingGameplayTag(Message) (Components|GetOwner) "(TagName=\"NPC.Status.NoHire\")"))
    (if _returnvalue
      (Variables|InteractionUI|SetDialogMarkerIndicatorSoftIcon (Variables|Default|GetDialogMarkerIndicator_Speech))
      (else
        (Variables|InteractionUI|SetDialogMarkerIndicatorSoftIcon (Variables|Default|GetDialogMarkerIndicator_Hire))))))
