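; Widget component that shows a passive-healing greeting bubble above a Pastry Heater NPC pawn.
; On non-authority clients only, BeginPlay raises the component, caches greeting text from the owner's PassiveHealer pawn component, then hides the bubble widget and seeds its preview message.
; A visibility-change binding reacts to HitTestInvisible by zeroing render opacity, waiting briefly, then setting the preview message again.
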(event EventBeginPlay
  (Utilities|SwitchHasAuthority
    (:Remote
      (Transformation|SetRelativeLocation "0, 0,150.000000")
      (bind _asfort_player_pawn (Utilities|Casting|CastToFortPlayerPawn (Components|GetOwner)))
      (Utilities|IsValid (Actor|GetComponentbyClass _asfort_player_pawn "/Script/FortniteGame.FortPawnComponent_PassiveHealer")
        (:Is%20Valid
          (Variables|Default|SetGreetingText (PassiveHealer|GetGreetingPassiveHealingText (Actor|GetComponentbyClass _asfort_player_pawn "/Script/FortniteGame.FortPawnComponent_PassiveHealer")))
          (Appearance|Event|BindEventtoOnVisibilityChanged (UserInterface|GetUserWidgetObject) (delegate Custom|CustomEvent_0))
          (Widget|SetVisibility (UserInterface|GetUserWidgetObject) "Hidden")
          (bind _asui_pastry_heater_message_bubble (Utilities|Casting|CastToUI_PastryHeater_MessageBubble (UserInterface|GetUserWidgetObject)))
          (Conversation|SetPreviewMessage _asui_pastry_heater_message_bubble (Variables|Default|GetGreetingText)))))))

(event Custom|CustomEvent_0
  (params (InVisibility :type ESlateVisibility%20Enum :default "Visible"))
  (bind _returnvalue (UserInterface|GetUserWidgetObject))
  (if (Utilities|Enum|Equal(Enum) InVisibility "HitTestInvisible")
    (Widget|SetRenderOpacity _returnvalue 0.0)
    (Utilities|FlowControl|Delay 0.5)
    (bind _asui_pastry_heater_message_bubble (Utilities|Casting|CastToUI_PastryHeater_MessageBubble _returnvalue))
    (Conversation|SetPreviewMessage _asui_pastry_heater_message_bubble (Variables|Default|GetGreetingText))))

(event EventTick
  (params (DeltaSeconds :type Float%20%28single-precision%29)))
